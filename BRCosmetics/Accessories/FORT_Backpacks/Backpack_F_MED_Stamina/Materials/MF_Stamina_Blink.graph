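# Material function for the Stamina backpack cosmetic that adds blinking emissive glow to incoming material attributes, toggled by the UseStaminaBlinking static switch.
# The SkinFX mask texture supplies dot and arrow masks in red and green, tinted by squared glow colours.
# A time-driven sine flickers one glow term, and the combined glow adds to existing emissive.
# A phase-offset threshold chain for limiting simultaneous dots appears unconnected to the output.
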MaterialFunction /BRCosmetics/Accessories/FORT_Backpacks/Backpack_F_MED_Stamina/Materials/MF_Stamina_Blink
graph {nodes [
  FunctionOutput(None=StaticSwitchParameter(True=NamedRerouteUsage(), False=NamedRerouteUsage()) {group=StaminaBlinking, parameterName=UseStaminaBlinking, sortPriority=0})
  n1 = Reroute() {desc=offset map}
  n2 = Divide(
   A=Time()
   B=ScalarParameter() {defaultValue=64, desc=time duration of the whole blinking animation, group=StaminaBlinking, parameterName=TimePhaseOffset}
  )
  n3 = ScalarParameter() {
   defaultValue=0.9
   desc=the higher the value the smaller is the amount of simultaneous dots
   group=StaminaBlinking
   parameterName=DotThreshold
   sliderMax=1
  }
  n4 = Subtract(A=Sine(None=Add(A=@n1, B=@n2)), B=@n3)
  Multiply(A=Saturate(None=Ceil(None=@n4)))
  n5 = TextureSample() {samplerType=SAMPLERTYPE_Masks, texture=/BRCosmetics/Accessories/FORT_Backpacks/Backpack_F_MED_Stamina/Textures/T_F_MED_Stamina_Backpack_FX}
  n6 = FunctionInput(Preview=@n5.RGBA) {bUsePreviewValueAsDefault=true, inputName=SkinFX, inputType=FunctionInput_Vector4, sortPriority=1}
  NamedRerouteDeclaration(None=ComponentMask(None=@n6) {r=true}) {name=DotMask, nodeColor=<rgba 0.6406 0 1 1>}
  NamedRerouteDeclaration(None=FunctionInput() {inputName=MA, inputType=FunctionInput_MaterialAttributes, sortPriority=1}) {name=MA, nodeColor=<rgba 0 0.2734 1 1>}
  NamedRerouteDeclaration(None=ComponentMask(None=@n6) {g=true}) {name=ArrowMask, nodeColor=<rgba 0 1 0.836 1>}
  n7 = GetMaterialAttributes(None=NamedRerouteUsage()) {attributeGetTypes [#1]}
  n8 = NamedRerouteUsage()
  n9 = Sine(None=Multiply(A=Time(), B=ScalarParameter() {defaultValue=1.22, group=StaminaBlinking, parameterName=FlickersPerSec, sliderMax=4}))
  n10 = Saturate(None=Add(A=Multiply(A=@n9) {constB=0.5}) {constB=0.5})
  n11 = VectorParameter() {group=StaminaBlinking, parameterName=DotGlowColor, sortPriority=2, defaultValue=<rgba 1 0.5772 0 2>}
  n12 = VectorParameter() {group=StaminaBlinking, parameterName=ArrowGlowColor, sortPriority=2, defaultValue=<rgba 0.053 1 0 1.1>}
  n13 = Add(
   A=Multiply(A=Multiply(A=@n8, B=@n10), B=Multiply(A=@n11.RGB, B=@n11.RGB))
   B=Multiply(A=NamedRerouteUsage(), B=Multiply(A=@n12.RGB, B=@n12.RGB))
  )
  NamedRerouteDeclaration(None=SetMaterialAttributes!(MaterialAttributes=@n7.MaterialAttributes, EmissiveColor=Add(A=@n7.EmissiveColor, B=@n13)) {attributeSetTypes [#1]}) {name=MA Output, nodeColor=<rgba 0 1 0.0156 1>}
 ]}
