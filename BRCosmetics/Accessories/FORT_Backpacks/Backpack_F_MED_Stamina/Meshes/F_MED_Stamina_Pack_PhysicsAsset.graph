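# Physics asset for the F_MED_Stamina backpack cosmetic mesh, defining collision and joint limits for its dangling strap.
# The pack bone carries two capsules, and six dyn_strap bones each carry a small capsule.
# Constraints link pack through the straps in a single chain, with narrow swing limits that are tightest at the root and widest mid-chain.
# Every joint has a zero twist limit.
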PhysicsAsset /BRCosmetics/Accessories/FORT_Backpacks/Backpack_F_MED_Stamina/Meshes/F_MED_Stamina_Pack_PhysicsAsset
bodies {
 dyn_strap03 [{shapeType=Capsule, length=1.3847, radius=1.3029, shapeName=dyn_strap03_capsule, center=<-0.2099 0.0625 0>, rotation=<rot -90 0 7.857>}]
 dyn_strap04 [{
   shapeType=Capsule
   length=2.2297
   radius=1.5583
   shapeName=dyn_strap04_capsule
   center=<0.9488 0.0958 0>
   rotation=<rot 90 179.5921 176.7799>
  }]
 dyn_strap05 [{
   shapeType=Capsule
   length=1.4884
   radius=2.0116
   shapeName=dyn_strap05_capsule
   center=<2.5599 0.0062 0.0294>
   rotation=<rot 88.1023 -179.9993 179.9997>
  }]
 dyn_strap_00 [{shapeType=Capsule, length=0, radius=0.75, shapeName=dyn_strap_00_capsule, center=<0.3489 0.0608 0>, rotation=<rot 90 160.529 160.5295>}]
 dyn_strap_01 [{
   shapeType=Capsule
   length=1.3742
   radius=0.7325
   shapeName=dyn_strap_01_capsule
   center=<-0.2535 -0.0268 0>
   rotation=<rot -90 169.3803 -157.135>
  }]
 dyn_strap_02 [{
   shapeType=Capsule
   length=1.1783
   radius=0.9532
   shapeName=dyn_strap_02_capsule
   center=<0.0539 0.0468 0>
   rotation=<rot 90 152.2329 139.4872>
  }]
 pack [
  {shapeType=Capsule, length=0, radius=8.8094, shapeName=pack_capsule, center=<-0.0383 0.8481 -1.5405>, rotation=<rot 0 0 90>}
  {shapeType=Capsule, length=9.1883, radius=5, shapeName=pack_capsule, center=<-2.728 16.6182 0>, rotation=<rot 90 0 0>}
 ]
}
constraints [
 {bone1Name=pack, bone2Name=dyn_strap_00, swing1LimitDegrees=5, swing2LimitDegrees=5, twistLimitDegrees=0}
 {bone1Name=dyn_strap_00, bone2Name=dyn_strap_01, swing1LimitDegrees=10, swing2LimitDegrees=10, twistLimitDegrees=0}
 {bone1Name=dyn_strap_01, bone2Name=dyn_strap_02, swing1LimitDegrees=20, swing2LimitDegrees=20, twistLimitDegrees=0}
 {bone1Name=dyn_strap_02, bone2Name=dyn_strap03, swing1LimitDegrees=25, swing2LimitDegrees=25, twistLimitDegrees=0}
 {bone1Name=dyn_strap03, bone2Name=dyn_strap04, swing1LimitDegrees=15, swing2LimitDegrees=15, twistLimitDegrees=0}
 {bone1Name=dyn_strap04, bone2Name=dyn_strap05, swing1LimitDegrees=15, swing2LimitDegrees=15, twistLimitDegrees=0}
]
